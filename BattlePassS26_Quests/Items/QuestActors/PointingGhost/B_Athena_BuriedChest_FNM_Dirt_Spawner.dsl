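; Spawner Blueprint for the buried-chest dirt prop used by the Battle Pass S26 pointing ghost quest.
; SpawnChest casts the given class to the Athena buried chest dirt prop class and spawns it at the given transform.
; The spawn passes the GhostToHide variable into the new actor, presumably so the chest can hide that ghost.
; The construction script only calls its parent.
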(fn SpawnChest
  (params (Class :type Actor%20Class%20Reference) (SpawnTransform :type Transform))
  (bind _asprop_athena_buried_chest_fnm_dirt (Utilities|Casting|CastToProp_Athena_BuriedChest_FNM_DirtClass Class))
  (Game|SpawnActorfromClass _asprop_athena_buried_chest_fnm_dirt SpawnTransform "Undefined" "MultiplyWithRoot" 0 (Variables|Default|GetGhostToHide) 0 0 0))

(fn ConstructionScript
  (|CallParentFunction))
