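; Buried chest dirt prop for the Season 26 Battle Pass pointing ghost quest, which clears its glow and linked ghost once fully dug up.
; On max dug, it runs the parent logic, destroys the FNM glow VFX component, then hides the referenced ghost and gives it a very short life span.
; It also casts that ghost to S26_PointingGhost to fade out its ambient audio.
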(event EventBeginPlay
  (|CallParentFunction) ; This node is disabled and will not be called.\nDrag off pins to build functionality.
)

(event Collision|EventActorBeginOverlap
  (params (OtherActor :type Actor%20Object%20Reference)))

(event EventTick
  (params (DeltaSeconds :type Float%20%28single-precision%29)))

(event EventOnMaxDug
  (bind _ghosttohide (Variables|Default|GetGhostToHide))
  (|CallParentFunction)
  (Components|DestroyComponent (Variables|Default|GetFNM%20Glow%20VFX))
  (Rendering|SetActorHiddenInGame _ghosttohide true)
  (Actor|SetLifeSpan _ghosttohide 0.1)
  (bind _ass26_pointing_ghost (Utilities|Casting|CastToS26_PointingGhost _ghosttohide))
  (Audio|Components|Audio|FadeOut (Variables|/BattlePassS26_Quests/Items/QuestActors/PointingGhost/S26_PointingGhost.S26_PointingGhost_C|GetPointingGhostAmbAudio _ass26_pointing_ghost) 0.2 0.0 "Linear" :!class /Script/Engine.AudioComponent))

(fn ConstructionScript
  (|CallParentFunction))
