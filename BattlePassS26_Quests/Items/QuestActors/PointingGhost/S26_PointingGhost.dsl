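; Pirate ghost quest actor for Battle Pass S26 that serves as a pointing ghost prop with ambient sound.
; On BeginPlay, where cosmetic events can run, it sets net dormancy and attaches a stored looping-ambience component to the body mesh.
; A low-end platform pass, gated by mobile and effects-quality checks, swaps body and head materials to mostly opaque variants.
; EndPlay stops the ambient audio.
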(event EventBeginPlay
  (bind _sm_med_pirate_t4_head (Variables|Default|GetSM_MED_Pirate_T4_head))
  (bind _sm_med_pirate_t4_body (Variables|Default|GetSM_MED_Pirate_T4_Body))
  (Networking|CanExecuteCosmeticEvents
    (:True
      (Utilities|FlowControl|Sequence
        (:then_0
          (Networking|SetNetDormancy "DORM_DormantAll")
          (bind _returnvalue (Audio|SpawnSoundAttached (Variables|Default|GetSound) (Variables|Default|GetSM_MED_Pirate_T4_Body)))
          (Variables|Default|SetPointingGhostAmbAudio _returnvalue))
        (:then_1
          ;[0.3214 0.7978 1] Low end platform material setup
          (if (not (TouchPlatform|IsMobileGame))
            (Utilities|EffectsQualityCheck 0
              (:Medium
                _)
              (:Low
                _))
            (Rendering|Material|SetMaterial _sm_med_pirate_t4_body 0 "/BattlePassS26_Quests/Items/QuestActors/PointingGhost/material/MI_GhostPointing_LegHood_opaque.MI_GhostPointing_LegHood_opaque")
            (Rendering|Material|SetMaterial _sm_med_pirate_t4_body 1 "/BattlePassS26_Quests/Items/QuestActors/PointingGhost/material/MI_GhostPointing_Torso_opaque.MI_GhostPointing_Torso_opaque")
            (Rendering|Material|SetMaterial _sm_med_pirate_t4_body 2 "/BattlePassS26_Quests/Items/QuestActors/PointingGhost/material/MI_GhostPointing_Seaweed.MI_GhostPointing_Seaweed")
            (Rendering|Material|SetMaterial _sm_med_pirate_t4_head 0 "/BattlePassS26_Quests/Items/QuestActors/PointingGhost/material/MI_GhostPointing_Hood_opaque.MI_GhostPointing_Hood_opaque")
            (Rendering|Material|SetMaterial _sm_med_pirate_t4_head 1 "/BattlePassS26_Quests/Items/QuestActors/PointingGhost/material/MI_GhostPointing_Head_opaque.MI_GhostPointing_Head_opaque")))))))

(event EventEndPlay
  (params (EndPlayReason :type EEndPlayReason%20Enum :default "Destroyed"))
  (Audio|Components|Audio|Stop (Variables|Default|GetPointingGhostAmbAudio)))

(fn ConstructionScript)
