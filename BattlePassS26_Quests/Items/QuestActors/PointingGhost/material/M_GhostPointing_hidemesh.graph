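# Unlit masked material for the Pointing Ghost quest actor in Battle Pass S26, named as a mesh-hiding material.
# Its graph holds a single Constant at its default value of zero, with no textures or parameters.
# It is flagged for skeletal meshes, clothing and mesh deformers, so it likely blanks out parts of the ghost's mesh through masking.
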Material /BattlePassS26_Quests/Items/QuestActors/PointingGhost/material/M_GhostPointing_hidemesh
graph {
 properties {
  bUsedWithClothing=true
  bUsedWithMeshDeformer=true
  bUsedWithSkeletalMesh=true
  blendMode=BLEND_Masked
  pixelDepthOffsetMode=PDOM_Legacy
  shadingModel=MSM_Unlit
  usedShadingModels=Unlit
 }
 nodes [Constant()]
}
