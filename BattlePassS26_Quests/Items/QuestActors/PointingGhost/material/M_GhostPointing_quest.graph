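# Unlit translucent material for the pointing ghost quest actor, giving a skeletal mesh a smoky, blue-rimmed spectral look.
# Emissive combines fresnel tint, desaturated base texture and world-aligned scrolling smoke.
# Opacity combines fresnel and panning smoke with actor-relative directional gradients that fade the body, plus a custom depth test with camera-distance bias that hides occluded layers.
# World position offset adds wind and cloth motion, blended along the same direction.
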Material /BattlePassS26_Quests/Items/QuestActors/PointingGhost/material/M_GhostPointing_quest
graph {
 properties {
  allowTranslucentCustomDepthWrites=true
  bCastRayTracedShadows=false
  bUseMaterialAttributes=true
  bUseTranslucencyVertexFog=false
  bUsedWithClothing=true
  bUsedWithMeshDeformer=true
  bUsedWithMorphTargets=true
  bUsedWithSkeletalMesh=true
  blendMode=BLEND_Translucent
  pixelDepthOffsetMode=PDOM_Legacy
  shadingModel=MSM_Unlit
  translucencyDirectionalLightingIntensity=2
  translucencyLightingMode=TLM_VolumetricPerVertexDirectional
  translucentShadowDensityScale=10
  usedShadingModels=Unlit
 }
 nodes [
  Constant() {r=1}
  n1 = PixelDepth()
  LinearInterpolate(Alpha=Saturate(None=Divide(A=@n1.R) {constB=6000})) {constA=-0.01, constB=-2}
  n2 = Multiply(
   A=Desaturation(
    None=TextureSampleParameter2D() {parameterName=Normal, samplerType=SAMPLERTYPE_Normal, texture=&FlatNormalBlue}.RGB
    Fraction=Constant() {r=1}
   )
   B=ScalarParameter() {parameterName=NormnalBlendIndensity}
  ) {constB=0.02}
  n3 = Constant3Vector() {constant=<rgba 0.0218 0.0221 0.026 1>}
  n4 = Constant3Vector() {constant=<rgba 0.3923 0.4247 0.7 1>}
  n5 = MaterialFunctionCall(Brightness=Constant() {r=0.65}, Exponent=Constant() {r=3}) {materialFunction=&MF_Fresnel_FX}
  n6 = Saturate(None=Power(Base=NamedRerouteUsage(), Exp=Constant() {r=5}) {constExponent=10})
  n7 = Constant() {r=0.45}
  n8 = Power(
   Base=Desaturation(
    None=TextureSampleParameter2D() {parameterName=Base Color, samplerType=SAMPLERTYPE_LinearColor, texture=/Game/Global/Textures/Default/Blanks/T_BlackLinearColor}.RGB
    Fraction=Constant() {r=0.25}
   )
   Exp=Constant() {r=1.3}
  ) {constExponent=4}
  n9 = Constant3Vector() {constant=<rgba 0.02 0.02 0.02 0>}
  n10 = ScalarParameter() {defaultValue=0.2, parameterName=BlackColorBlend}
  n11 = Add(
   A=LinearInterpolate(A=LinearInterpolate(A=@n3, B=@n4, Alpha=@n5.Fresnel Double Sided), Alpha=Constant() {r=0.9}) {constB=0}
   B=Multiply(A=LinearInterpolate(A=@n6, Alpha=@n7), B=LinearInterpolate(A=@n8, B=@n9, Alpha=@n10) {constB=0})
  )
  n12 = Multiply(A=Time(), B=Constant() {r=30})
  n13 = Constant3Vector() {constant=<rgba 0 0 -1 0>}
  n14 = MaterialFunctionCall(
   TextureObject=TextureObject() {texture=/Game/Packages/Fortress_Transylvania/Transylvania_effects/Textures/T_Smoke2_Tile}
   TextureSize=Constant() {r=40}
   WorldPosition=Add(A=WorldPosition().XYZ, B=Multiply(A=@n12, B=@n13))
  ) {materialFunction=&WorldAlignedTexture}
  n15 = Multiply(A=Saturate(None=Power(Base=@n14.XY Texture) {constExponent=5}), B=Constant3Vector() {constant=<rgba 10 10 10 1>})
  n16 = ScalarParameter() {defaultValue=0.1, parameterName=blueSmoke Blend}
  MakeMaterialAttributes!(
   BaseColor
   Metallic
   Specular
   Roughness
   Anisotropy
   EmissiveColor=Add(A=Multiply(A=@n15, B=@n16), B=Add(A=@n2, B=@n11))
   Opacity=NamedRerouteUsage()
   OpacityMask
   Normal
   Tangent
   WorldPositionOffset=NamedRerouteUsage()
   SubsurfaceColor
   ClearCoat
   ClearCoatRoughness
   AmbientOcclusion
   Refraction
   CustomizedUVs_0
   CustomizedUVs_1
   CustomizedUVs_2
   CustomizedUVs_3
   CustomizedUVs_4
   CustomizedUVs_5
   CustomizedUVs_6
   CustomizedUVs_7
   PixelDepthOffset
   ShadingModel
   Displacement
  )
  n17 = TextureSample(UVs=Panner(Coordinate=TextureCoordinate() {uTiling=4}) {speedX=0.15, speedY=0.18}) {samplerSource=SSM_Wrap_WorldGroupSettings, texture=/Game/Global/Textures/Default/Tileable_Patterns/T_Blowing_Smoke_Pattern2_Tile_DISP}
  n18 = TextureSample(UVs=Panner(Coordinate=TextureCoordinate() {uTiling=8}) {speedX=-0.3, speedY=-0.2}) {samplerSource=SSM_Wrap_WorldGroupSettings, texture=&T_Blowing_Smoke_Pattern2_Tile_DISP}
  n19 = Multiply(A=Add(A=@n17.R, B=@n18.R)) {constB=0.3}
  n20 = WorldPosition()
  n21 = Constant3Vector() {constant=<rgba 0 0 -1 0>}
  n22 = Multiply(A=Multiply(A=Time(), B=ScalarParameter() {defaultValue=60, parameterName=SmokeTextureSpeed1}), B=@n21)
  n23 = WorldPosition()
  n24 = Multiply(A=Multiply(A=Time(), B=ScalarParameter() {defaultValue=30, parameterName=SmokeTexturSpeed2}), B=@n21)
  n25 = Add(
   A=MaterialFunctionCall(
    TextureObject=TextureObject() {texture=/Game/Effects/Fort_Effects/Textures/Tile/T_Inky_Smoke_Tile_Chromatic}
    TextureSize=ScalarParameter() {defaultValue=80, parameterName=SmokeTexturesize 1}
    WorldPosition=Add(A=@n20.XYZ, B=@n22)
   ) {materialFunction=&WorldAlignedTexture}.XY Texture
   B=MaterialFunctionCall(
    TextureObject=TextureObject() {texture=&T_Blowing_Smoke_Pattern2_Tile_DISP}
    TextureSize=ScalarParameter() {defaultValue=60, parameterName=SmokeTexturesize 2}
    WorldPosition=Add(A=@n23.XYZ, B=@n24)
   ) {materialFunction=&WorldAlignedTexture}.XY Texture
  )
  NamedRerouteDeclaration(None=Saturate(None=Add(A=@n19, B=@n25))) {name=Smoke 2, nodeColor=<rgba 1 0.7969 0 1>}
  n26 = StaticBool()
  n27 = MaterialFunctionCall!(
   UVs
   Pan Speed
   WPO_ClothOffsetDirection
   WPO_ClothNormalIntensity
   WPO_ClothOffsetIntensity
   "Motion Mask (Defaults to the Red Vertex Color)"
   WPO_ClothWPOWindTextureOffset
   Pan Direction
  ) {materialFunction=/Game/Packages/Fortress_SharedMaterials/Base_Material_Functions/Cloth}
  n28 = Multiply(A=@n27.Additive WPO, B=ScalarParameter() {defaultValue=0.3, parameterName=WPOBlend}) {constB=0.2}
  n29 = Multiply(A=@n27.Additive WPO, B=ScalarParameter() {defaultValue=0.02, parameterName=WPOBlend2}) {constB=0.02}
  n30 = DotProduct(
   A=Subtract(A=WorldPosition().XYZ, B=ActorPositionWS())
   B=VectorParameter() {parameterName=DissolveGradientDirectionalVector, defaultValue=<rgba 0 0 1 1>}.RGB
  )
  n31 = Constant() {r=120}
  n32 = Saturate(None=Power(Base=Divide(A=@n30, B=@n31) {constB=30}, Exp=Constant() {r=1}))
  NamedRerouteDeclaration(None=Add(
    A=MaterialFunctionCall!(Tangent Space Normal, UseWind=@n26) {materialFunction=/Game/Packages/Fortress_SharedMaterials/Base_Material_Functions/BaseShaderWindFunction}.World position offset
    B=LinearInterpolate(A=@n28, B=@n29, Alpha=@n32)
   )) {name=WPO, nodeColor=<rgba 1 0 0.8672 1>}
  n33 = Subtract(A=WorldPosition().XYZ, B=ActorPositionWS())
  n34 = VectorParameter() {parameterName=DissolveGradientDirectionalVector, defaultValue=<rgba 0 0 1 1>}
  n35 = Power(
   Base=Divide(A=DotProduct(A=@n33, B=@n34.RGB), B=ScalarParameter() {defaultValue=90, parameterName=gradation_distance}) {constB=30}
   Exp=Constant() {r=3}
  )
  n36 = Add(
   A=MaterialFunctionCall(
    Brightness=ScalarParameter() {defaultValue=0.6224, parameterName=Fresnel Brightness}
    Exponent=ScalarParameter() {defaultValue=3, parameterName=Fresnel Exponent}
   ) {materialFunction=&MF_Fresnel_FX}.Fresnel
   B=Multiply(A=NamedRerouteUsage(), B=ScalarParameter() {defaultValue=15, parameterName=Secondary Smoke})
  )
  n37 = ScalarParameter() {defaultValue=0.1, parameterName=Opacity}
  n38 = Add(
   A=Multiply(A=Clamp(None=@n35), B=ScalarParameter() {defaultValue=0.7, parameterName=gradation_indensity})
   B=NamedRerouteDeclaration(None=Multiply(A=@n36, B=@n37)) {name=Smoke aniamtion, nodeColor=<rgba 0.8984 0 1 1>}
  )
  n39 = Subtract(A=WorldPosition().XYZ, B=ActorPositionWS())
  n40 = VectorParameter() {parameterName=DissolveGradientDirectionalVector, defaultValue=<rgba 0 0 1 1>}
  n41 = Clamp(None=Divide(A=DotProduct(A=@n39, B=@n40.RGB), B=ScalarParameter() {defaultValue=120, parameterName=DissolveGradientDistance}) {constB=30})
  NamedRerouteDeclaration(None=LinearInterpolate(A=Multiply(A=@n38, B=@n41), Alpha=ScalarParameter() {parameterName=Opacity_indensity})) {name=Opacity, nodeColor=<rgba 1 0 0.9844 1>}
  n42 = NamedRerouteUsage()
  n43 = Divide(A=Distance(A=CameraPositionWS(), B=ActorPositionWS())) {constB=1000}
  n44 = Add(
   A=PixelDepth().R
   B=LinearInterpolate(
    A=ScalarParameter() {defaultValue=-0.25, parameterName=Bias_Close, sliderMax=1, sliderMin=-1}
    B=ScalarParameter() {defaultValue=-1, parameterName=Bias_Far, sliderMax=1, sliderMin=-1}
    Alpha=Saturate(None=@n43)
   )
  )
  n45 = ComponentMask(None=SceneTexture() {sceneTextureId=PPI_CustomDepth}.Color) {r=true}
  n46 = Constant()
  n47 = Constant() {r=1}
  NamedRerouteDeclaration(None=Multiply(A=Saturate(None=@n42), B=If(A=@n44, B=@n45, "A > B"=@n46, "A == B"=@n47, "A < B"=@n47))) {name=Final Opacity, nodeColor=<rgba 1 0.9375 0 1>}
 ]
}
