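# Translucent unlit ghost material for the Season 26 pointing-ghost pirate quest actor, built on the Pirate Progressive head mask and normal textures.
# Fresnel-warped, panning spooky-face and caustic textures form a detail mask that drives dark-to-bright emissive color and opacity.
# Vertex colors, inverted Fresnel and depth fades shape opacity, with a deferred-only scene-normal glow at intersections.
# Wind and camera offset move the vertices.
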Material /BattlePassS26_Quests/Items/QuestActors/PointingGhost/material/M_MED_Pirate_Progressive_FX_Quest
graph {
 properties {
  bUsedWithMeshDeformer=true
  bUsedWithSkeletalMesh=true
  blendMode=BLEND_Translucent
  pixelDepthOffsetMode=PDOM_Legacy
  shadingModel=MSM_Unlit
  usedShadingModels=Unlit
 }
 nodes [
  Add(
   A=MaterialFunctionCall!(Tangent Space Normal, UseWind=StaticBool() {value=true}) {materialFunction=/Game/Packages/Fortress_SharedMaterials/Base_Material_Functions/BaseShaderWindFunction}.World position offset
   B=MaterialFunctionCall(Offset Amount=ScalarParameter() {defaultValue=2, parameterName=CamOfsset}) {materialFunction=/Engine/Functions/Engine_MaterialFunctions02/WorldPositionOffset/CameraOffset}.Camera Offset
  )
  n1 = VectorParameter() {parameterName=DarkColor, defaultValue=<rgba 0.125 0.3477 0.6 0>}
  n2 = Multiply(A=ScalarParameter() {defaultValue=5, parameterName=ColorMult}, B=VectorParameter() {parameterName=Color, defaultValue=<rgba 10 10 10 1>}.RGB)
  n3 = TextureSample() {
   samplerType=SAMPLERTYPE_Masks
   texture=/Game/Characters/Player/Male/Medium/Bodies/M_MED_Pirate_Progressive/Textures/T_M_MED_Pirate_Progressive_Head_123_M
  }
  n4 = OneMinus(None=@n3.G)
  n5 = VertexColor()
  n6 = Multiply(
   A=ScalarParameter() {defaultValue=1, parameterName=TextureTile}
   B=LinearInterpolate(A=Constant2Vector() {g=0.35, r=1.25}, B=Constant() {r=1}, Alpha=@n5.B) {constA=0.25, constB=1.25}
  )
  n7 = Multiply(A=TextureCoordinate() {uTiling=5.5, vTiling=3.6}, B=@n6)
  n8 = LinearInterpolate(Alpha=@n5.B) {constA=0.7}
  n9 = Multiply(A=Sine(None=Time()) {period=4}) {constB=0.12}
  n10 = TextureSample() {
   samplerType=SAMPLERTYPE_Normal
   texture=/Game/Characters/Player/Male/Medium/Bodies/M_MED_Pirate_Progressive/Textures/T_M_MED_Pirate_Progressive_Head_123_N
  }
  n11 = Transform(None=@n10.RGB)
  n12 = Fresnel(ExponentIn=Multiply(A=Add(A=@n8, B=@n9), B=ScalarParameter() {defaultValue=0.4, parameterName=CausticsFresnelExp}), Normal=@n11) {baseReflectFraction=0, exponent=0.7}
  n13 = Multiply(A=@n12) {constB=0.1}
  n14 = Multiply(A=@n13) {constB=-0.8}
  n15 = Multiply(A=Multiply(A=Time(), B=@n8), B=ScalarParameter() {defaultValue=1, parameterName=PanSpeed})
  n16 = TextureSample(UVs=Panner(Coordinate=Add(A=@n7, B=@n14), Time=@n15) {speedX=-0.01, speedY=0.088}) {texture=/Game/Effects/Fort_Effects/Textures/Tile/T_SpookyFaces_Chroma_Tile}
  n17 = Constant() {r=-0.3}
  n18 = Add(A=@n13, B=Multiply(A=@n6, B=TextureCoordinate() {uTiling=2, vTiling=1.7}))
  n19 = LinearInterpolate(Alpha=TextureSample(UVs=Panner(Coordinate=@n18, Time=@n15) {speedX=0.01, speedY=0.06}) {texture=/Game/Global/Textures/Default/Caustics/T_Caustics_Softer}.G) {constA=0.4, constB=0.88}
  n20 = Clamp(None=SphereMask(A=PixelDepth().R, B=Constant(), Radius=ScalarParameter() {defaultValue=500, parameterName=DetailFadeDistance}) {hardnessPercent=0})
  n21 = Clamp(None=Power(Base=Multiply(A=@n12) {constB=1.5}) {constExponent=3})
  n22 = DepthFade() {fadeDistanceDefault=80}
  n23 = ScalarParameter() {defaultValue=2, parameterName=WorldNormalFresnelExp}
  n24 = ComponentMask(None=SceneTexture() {sceneTextureId=PPI_WorldNormal}.Color) {b=true, g=true, r=true}
  n25 = Multiply(A=Multiply(A=OneMinus(None=@n22), B=Fresnel(ExponentIn=@n23, Normal=@n24) {baseReflectFraction=0, exponent=3})) {constB=0.2}
  n26 = Constant()
  n27 = ShadingPathSwitch(Default=@n25, Deferred=@n25, Forward=@n26, Mobile=@n26)
  n28 = Add(A=LinearInterpolate(A=Desaturation(None=@n16.RGB, Fraction=@n17), B=LinearInterpolate(B=@n19, Alpha=@n20) {constA=0.5}, Alpha=@n21), B=@n27) {constB=0}
  n29 = Multiply(A=@n3.R, B=Power(Base=Multiply(A=@n4, B=@n28), Exp=LinearInterpolate(Alpha=@n21) {constA=1.8, constB=0.8}))
  n30 = Clamp(None=@n29)
  MaterialFunctionCall(Emissive=Multiply(A=LinearInterpolate(A=@n1.RGB, B=@n2, Alpha=@n30), B=@n29)) {materialFunction=&MF_FX_EmissiveAdjustments}
  n31 = VertexColor()
  n32 = Add(A=@n29, B=Multiply(A=LinearInterpolate(Alpha=@n31.B) {constA=0.5}, B=ScalarParameter() {defaultValue=0.5, parameterName=DarknessAdd}))
  n33 = OneMinus(None=Fresnel(ExponentIn=ScalarParameter() {defaultValue=0.4, parameterName=OpacityFresnelExp}, Normal=@n11) {baseReflectFraction=0})
  n34 = Multiply(A=@n33, B=@n33)
  n35 = ScalarParameter() {defaultValue=8, parameterName=BodyOpacityMult}
  n36 = OneMinus(None=VertexColor() {desc=Fade out nose/mouth}.R)
  n37 = Add(A=Clamp(None=Multiply(A=@n34, B=@n35) {constB=3}, Max=Power(Base=@n36) {constExponent=3}), B=@n27)
  n38 = Multiply(A=LinearInterpolate(Alpha=@n20) {constA=8}, B=Add(A=ScalarParameter() {defaultValue=1, parameterName=FadeDistance}, B=Multiply(A=@n31.B)))
  Multiply(A=Multiply(A=Multiply(A=@n32, B=@n37), B=DepthFade(FadeDistance=@n38)), B=@n31.G)
 ]
}
